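# Sound cue for the OilRig boss's suspicious-state voice lines, from the S32 VO set.
# A single Random node picks among six equally weighted wave players, each holding one Suspicious line.
# It uses the shared NPC alerted VO attenuation and the critical enemy VO sound class at reduced volume.
# A boss combat chatter concurrency set governs overlap with other boss barks.
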SoundCue /Clyde/Athena/Audio/VO/Bosses/OilRig/Cues/VO_OilRig_S32_Suspicious_Cue
root=Random_0
nodes [
 {id=Random_0, children [WavePlayer_0,WavePlayer_1,WavePlayer_2,WavePlayer_3,WavePlayer_4,WavePlayer_5], properties {weights [1,1,1,1,1,1]}}
 {id=WavePlayer_0, properties {soundWaveAssetPtr=/Clyde/Athena/Audio/VO/Bosses/OilRig/Waves/VO_OilRig_S32_Suspicious_010}}
 {id=WavePlayer_1, properties {soundWaveAssetPtr=/Clyde/Athena/Audio/VO/Bosses/OilRig/Waves/VO_OilRig_S32_Suspicious_020}}
 {id=WavePlayer_2, properties {soundWaveAssetPtr=/Clyde/Athena/Audio/VO/Bosses/OilRig/Waves/VO_OilRig_S32_Suspicious_030}}
 {id=WavePlayer_3, properties {soundWaveAssetPtr=/Clyde/Athena/Audio/VO/Bosses/OilRig/Waves/VO_OilRig_S32_Suspicious_040}}
 {id=WavePlayer_4, properties {soundWaveAssetPtr=/Clyde/Athena/Audio/VO/Bosses/OilRig/Waves/VO_OilRig_S32_Suspicious_050}}
 {id=WavePlayer_5, properties {soundWaveAssetPtr=/Clyde/Athena/Audio/VO/Bosses/OilRig/Waves/VO_OilRig_S32_Suspicious_060}}
]
properties {
 attenuationSettings=/Game/Sounds/Fort_Human_VO_BR/NPCs/Shared/NPC_VO_Alerted_Att
 maxDistance=12500
 soundClassObject=&Fort_VO_Enemy_Critical
 volumeMultiplier=0.7166
 concurrencySet [&NPC_Boss_Combat_Chatter_Conc]
}
